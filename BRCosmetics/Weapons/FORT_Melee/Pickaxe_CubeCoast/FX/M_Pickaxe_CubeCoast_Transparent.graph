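# Unlit translucent material for the CubeCoast pickaxe, named as its transparent variant.
# It is enabled for skeletal meshes and mesh deformers.
# The graph holds only a default Constant and a TextureCoordinate node, with no connections to each other or to any material output shown.
# It has no textures or parameters.
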Material /BRCosmetics/Weapons/FORT_Melee/Pickaxe_CubeCoast/FX/M_Pickaxe_CubeCoast_Transparent
graph {
 properties {
  bUsedWithMeshDeformer=true
  bUsedWithSkeletalMesh=true
  blendMode=BLEND_Translucent
  pixelDepthOffsetMode=PDOM_Legacy
  shadingModel=MSM_Unlit
  usedShadingModels=Unlit
 }
 nodes [
  Constant()
  TextureCoordinate()
 ]
}
