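# Idle Niagara effect for the CubeCoast offhand pickaxe, rendering a bubble emitter with mesh and sprite renderers sampled from the left axe's skeletal mesh.
# A spring-driven pendulum simulates lag while render position stays on the sampled point; dynamic material parameters pass both, plus rest axis and previous position, to the material.
# An interpolated system value, apparently from User.bisSwinging, drives Y stretch and material curves; time of day scales color.
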NiagaraSystem /BRCosmetics/Weapons/FORT_Melee/Pickaxe_CubeCoast/FX/NS_Pickaxe_CubeCoast_Idle_Offhand
dependencies {
 usedDataInterfaces [&NiagaraDataInterfaceSkeletalMesh, &NiagaraDataInterfaceCurve]
 usedDynamicInputs [&MakeCustomFloatFromBool]
 usedModules [
  &SystemState
  &self:SystemSpawnScript.SystemScriptSource.SystemScriptGraph.NiagaraNodeAssignment_0.SetVariables_#1
  &InterpolateOverTime
 ]
}
emitters {BubbleMesh_L {
  emitterUpdateScript [
   &EmitterState(Life Cycle Mode=::NewEnumerator1, Loop Behavior=::NewEnumerator1, Loop Duration Mode=::NewEnumerator1, UseLoopDelay=false)
   &SpawnBurst_Instantaneous(Use Spawn Probability=false, UseLoopCountLimit=false)
  ]
  particleSpawnScript [
   &InitializeParticle(
    Color Mode=::NewEnumerator0
    Mesh Scale Mode=::NewEnumerator3
    Sprite Size Mode=::NewEnumerator3
    UsePositionOffset=false
    Write Material Random=true
   )
   &InitialMeshOrientation(Mesh Orientation Mode=::NewEnumerator2, Use Rotation Vector=false)
  ]
  particleUpdateScript [
   &ParticleState(Kill Particles When Lifetime Has Elapsed=false, Let Infinitely Lived Particles Die When Emitter Deactivates=true)
   &SkeletalMeshLocation(
    MaskSpawnGroup=false
    Mesh Sampling Type=::NewEnumerator5
    OverrideRandomSeed=false
    Position Sampling=::NewEnumerator2
    Skeletal Mesh=DataInterface() {propertyValues {
      DefaultMesh=null
      ExcludeBoneName=root
      OptionalComponentClass=null
      PreviewMesh=/BRCosmetics/Weapons/FORT_Melee/Pickaxe_CubeCoast/Meshes/CubeCoast_Axe_L
      SkinningMode=SkinOnTheFly
      SoftSourceActor=null
      SourceMode=Default
      UvSetIndex=0
      WholeMeshLOD=-1
      bExcludeBone=false
      bReadDeformedGeometry=true
      bRequireCurrentFrameData=true
      ComponentTags []
      FilteredBones [b_AR_Root]
      FilteredSockets [FX_Idle]
      MeshUserParameter {parameter {name=None, typeDefHandle {registeredTypeIndex=1}, varData []}}
      SamplingRegions []
     }}
   )
   &ScaleMeshSize(Scale Factor=&MakeVector(
     X=1
     Y=&FloatFromCurve(
      FloatCurve=DataInterface() {propertyValues {Curve {keys [
          {arriveTangent=1.2935, interpMode=RCIM_Cubic, leaveTangent=1.2935, tangentMode=RCTM_User, time=0, value=0.6}
          {arriveTangent=-0.0007, interpMode=RCIM_Cubic, leaveTangent=-0.0007, tangentMode=RCTM_User, time=1, value=2}
         ]}}}
      CurveIndex=$System.InterpolateOverTime.InterpolateOverTime_Float.Moving Average
     )
     Z=1
    ))
   &PendulumSetup(
    Module.Enable Spring Driver=true
    Pendulum Length=15
    Pendulum Pivot:Vector3f=$Particles.SkeletalMeshLocation.SampledPosition
    Pendulum Rest Axis:Vector3f=&TransformVector(Vector=<0 1 0>, Destination Space=::Simulation)
    "Spring Driven Constraint (Tightness)"=1
   )
   &SolveForcesAndVelocity(Clamp Velocity=false, Limit Acceleration=false)
   &PendulumConstraint()
   &CalculateAccurateVelocity()
   /SetVariables(Particles.PositionForRender:Vector3f=$Particles.SkeletalMeshLocation.SampledPosition) {isSetParameters=true}
   &DynamicMaterialParameters(
    Index 0 Param 1=&FloatFromCurve(
     FloatCurve=DataInterface() {propertyValues {Curve {keys [
         {arriveTangent=2.2304, interpMode=RCIM_Cubic, leaveTangent=2.2304, tangentMode=RCTM_User, time=0, value=0}
         {arriveTangent=0.4259, interpMode=RCIM_Cubic, leaveTangent=0.4259, tangentMode=RCTM_User, time=1, value=1}
        ]}}}
     CurveIndex=$System.InterpolateOverTime.InterpolateOverTime_Float.Moving Average
    )
    Index 0 Param 2=&RandomRangeFloat(Minimum=0, Maximum=100)
    Index 0 Param 3=0.4
    Index 0 Param 4=&FloatFromCurve(
     FloatCurve=DataInterface() {propertyValues {Curve {keys [{arriveTangent=11.5348, interpMode=RCIM_Cubic, leaveTangent=11.5348, tangentMode=RCTM_User, time=0, value=1}, <at 1 4>]}}}
     CurveIndex=$System.InterpolateOverTime.InterpolateOverTime_Float.Moving Average
    )
    Index 1 Param 1=&MakeFloatFromVector(Vector=Linked() {linkedVariable {name=Particles.Position, type=NiagaraPosition}}, Channel=::NewEnumerator0)
    Index 1 Param 2=&MakeFloatFromVector(Vector=Linked() {linkedVariable {name=Particles.Position, type=NiagaraPosition}}, Channel=::NewEnumerator1)
    Index 1 Param 3=&MakeFloatFromVector(Vector=Linked() {linkedVariable {name=Particles.Position, type=NiagaraPosition}}, Channel=::NewEnumerator2)
    Index 1 Param 4=&MakeFloatFromVector(Vector=$Transient.PhysicsPendulumRestAxis, Channel=::NewEnumerator0)
    Index 2 Param 1=&MakeFloatFromVector(Vector=$Transient.PhysicsPendulumRestAxis, Channel=::NewEnumerator1)
    Index 2 Param 2=&MakeFloatFromVector(Vector=$Transient.PhysicsPendulumRestAxis, Channel=::NewEnumerator2)
    Index 2 Param 3=&MakeFloatFromVector(Vector=$Particles.PositionForRender, Channel=::NewEnumerator0)
    Index 2 Param 4=&MakeFloatFromVector(Vector=$Particles.PositionForRender, Channel=::NewEnumerator1)
    Index 3 Param 1=&MakeFloatFromVector(Vector=$Particles.PositionForRender, Channel=::NewEnumerator2)
    Index 3 Param 2=&MakeFloatFromVector(Vector=Linked() {linkedVariable {name=Particles.Previous.Position, type=NiagaraPosition}}, Channel=::NewEnumerator0)
    Index 3 Param 3=&MakeFloatFromVector(Vector=Linked() {linkedVariable {name=Particles.Previous.Position, type=NiagaraPosition}}, Channel=::NewEnumerator1)
    Index 3 Param 4=&MakeFloatFromVector(Vector=Linked() {linkedVariable {name=Particles.Previous.Position, type=NiagaraPosition}}, Channel=::NewEnumerator2)
    Module.Param0WriteEnabled=true
    Module.Param0WriteEnabled1=true
    Module.Param0WriteEnabled2=true
    Module.Param0WriteEnabled3=true
    Module.Param1WriteEnabled=true
    Module.Param1WriteEnabled1=true
    Module.Param1WriteEnabled2=true
    Module.Param1WriteEnabled3=true
    Module.Param2WriteEnabled=true
    Module.Param2WriteEnabled1=true
    Module.Param2WriteEnabled2=true
    Module.Param2WriteEnabled3=true
    Module.Param3WriteEnabled=true
    Module.Param3WriteEnabled1=true
    Module.Param3WriteEnabled2=true
    Module.Param3WriteEnabled3=true
    Write Parameter Index 1=true
    Write Parameter Index 2=true
    Write Parameter Index 3=true
   )
   &ScaleColor(
    Scale Alpha=&FNTimeOfDayScalar(Night Scale=0.5, Morning Scale=1, Evening Scale=1)
    Scale RGB=&VectorFromFloat(Value=&FNTimeOfDayScalar(Night Scale=0.5, Morning Scale=1, Evening Scale=1))
    ScaleA=true
    ScaleRGB=true
   )
  ]
  rendererClasses [&NiagaraMeshRendererProperties, &NiagaraSpriteRendererProperties]
  renderers [
   {rendererClass=&NiagaraMeshRendererProperties, rendererIndex=0}
   {rendererClass=&NiagaraSpriteRendererProperties, rendererIndex=1}
  ]
 }}
summary {
 systemName=NS_Pickaxe_CubeCoast_Idle_Offhand
 emitters [{emitterName=BubbleMesh_L, rendererClasses [&NiagaraMeshRendererProperties, &NiagaraSpriteRendererProperties]}]
 userVariables [User.bisSwinging {type=NiagaraBool}]
}
